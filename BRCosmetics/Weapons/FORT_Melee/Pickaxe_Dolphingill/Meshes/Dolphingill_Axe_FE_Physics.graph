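# Physics asset for the Dolphingill pickaxe's FE axe mesh, providing bodies and joints for a flexible whip section on the two-handed handle.
# Sixteen whip capsules chain from the handle, each joint allowing a narrow swing with locked twist, ending in a fatter capsule at the tip.
# Separate thigh and calf capsules, joined per leg by free angular constraints, have no link to the whip and resemble collision proxies for character legs.
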PhysicsAsset /BRCosmetics/Weapons/FORT_Melee/Pickaxe_Dolphingill/Meshes/Dolphingill_Axe_FE_Physics
bodies {
 Melee_TwoHanded_Handle [{shapeType=Capsule, length=1, radius=1}]
 calf_l [{shapeType=Capsule, length=50, radius=8, center=<-30 0 0>, rotation=<rot 90 0 0>}]
 calf_r [{shapeType=Capsule, length=50, radius=10, center=<30 0.3257 0>, rotation=<rot 90 0 0>}]
 thigh_l [{shapeType=Capsule, length=40, radius=8, center=<-25 0 0>, rotation=<rot 90 0 0>}]
 thigh_r [{shapeType=Capsule, length=40, radius=10, center=<25 0 0>, rotation=<rot 90 0 0>}]
 whip_skin_01 [{shapeType=Capsule, length=2, radius=2, rotation=<rot 90 0 0>}]
 whip_skin_02 [{shapeType=Capsule, length=2, radius=2, rotation=<rot 90 0 0>}]
 whip_skin_03 [{shapeType=Capsule, length=2, radius=2, rotation=<rot 90 0 0>}]
 whip_skin_04 [{shapeType=Capsule, length=2, radius=2, rotation=<rot 90 0 0>}]
 whip_skin_05 [{shapeType=Capsule, length=2, radius=2, rotation=<rot 90 0 0>}]
 whip_skin_06 [{shapeType=Capsule, length=2, radius=2, rotation=<rot 90 0 0>}]
 whip_skin_07 [{shapeType=Capsule, length=2, radius=2, rotation=<rot 90 0 0>}]
 whip_skin_08 [{shapeType=Capsule, length=2, radius=2, rotation=<rot 90 0 0>}]
 whip_skin_09 [{shapeType=Capsule, length=2, radius=2, rotation=<rot 90 0 0>}]
 whip_skin_10 [{shapeType=Capsule, length=2, radius=2, rotation=<rot 90 0 0>}]
 whip_skin_11 [{shapeType=Capsule, length=2, radius=2, rotation=<rot 90 0 0>}]
 whip_skin_12 [{shapeType=Capsule, length=2, radius=2, rotation=<rot 90 0 0>}]
 whip_skin_13 [{shapeType=Capsule, length=2, radius=2, rotation=<rot 90 0 0>}]
 whip_skin_14 [{shapeType=Capsule, length=2, radius=2, rotation=<rot 90 0 0>}]
 whip_skin_15 [{shapeType=Capsule, length=2, radius=2, shapeName=whip_skin_15_capsule, rotation=<rot 90 0 0>}]
 whip_skin_16 [{shapeType=Capsule, length=2, radius=7, shapeName=whip_skin_16_capsule, center=<0.5 0 0>, rotation=<rot 90 0 0>}]
}
constraints [
 {
  bone1Name=whip_skin_01
  bone2Name=Melee_TwoHanded_Handle
  swing1LimitDegrees=5
  swing2LimitDegrees=5
  twistLimitDegrees=0
  twistMotion=Locked
 }
 {bone1Name=whip_skin_02, bone2Name=whip_skin_01, swing1LimitDegrees=5, swing2LimitDegrees=5, twistLimitDegrees=0, twistMotion=Locked}
 {bone1Name=whip_skin_03, bone2Name=whip_skin_02, swing1LimitDegrees=5, swing2LimitDegrees=5, twistLimitDegrees=0, twistMotion=Locked}
 {bone1Name=whip_skin_04, bone2Name=whip_skin_03, swing1LimitDegrees=5, swing2LimitDegrees=5, twistLimitDegrees=0, twistMotion=Locked}
 {bone1Name=whip_skin_05, bone2Name=whip_skin_04, swing1LimitDegrees=5, swing2LimitDegrees=5, twistLimitDegrees=0, twistMotion=Locked}
 {bone1Name=whip_skin_06, bone2Name=whip_skin_05, swing1LimitDegrees=5, swing2LimitDegrees=5, twistLimitDegrees=0, twistMotion=Locked}
 {bone1Name=whip_skin_07, bone2Name=whip_skin_06, swing1LimitDegrees=5, swing2LimitDegrees=5, twistLimitDegrees=0, twistMotion=Locked}
 {bone1Name=whip_skin_08, bone2Name=whip_skin_07, swing1LimitDegrees=5, swing2LimitDegrees=5, twistLimitDegrees=0, twistMotion=Locked}
 {bone1Name=whip_skin_09, bone2Name=whip_skin_08, swing1LimitDegrees=5, swing2LimitDegrees=5, twistLimitDegrees=0, twistMotion=Locked}
 {bone1Name=whip_skin_10, bone2Name=whip_skin_09, swing1LimitDegrees=5, swing2LimitDegrees=5, twistLimitDegrees=0, twistMotion=Locked}
 {bone1Name=whip_skin_11, bone2Name=whip_skin_10, swing1LimitDegrees=5, swing2LimitDegrees=5, twistLimitDegrees=0, twistMotion=Locked}
 {bone1Name=whip_skin_12, bone2Name=whip_skin_11, swing1LimitDegrees=5, swing2LimitDegrees=5, twistLimitDegrees=0, twistMotion=Locked}
 {bone1Name=whip_skin_13, bone2Name=whip_skin_12, swing1LimitDegrees=5, swing2LimitDegrees=5, twistLimitDegrees=0, twistMotion=Locked}
 {bone1Name=whip_skin_14, bone2Name=whip_skin_13, swing1LimitDegrees=5, swing2LimitDegrees=5, twistLimitDegrees=0, twistMotion=Locked}
 {
  bone1Name=calf_l
  bone2Name=thigh_l
  swing1LimitDegrees=45
  swing1Motion=Free
  swing2LimitDegrees=45
  swing2Motion=Free
  twistLimitDegrees=45
  twistMotion=Free
 }
 {
  bone1Name=calf_r
  bone2Name=thigh_r
  swing1LimitDegrees=45
  swing1Motion=Free
  swing2LimitDegrees=45
  swing2Motion=Free
  twistLimitDegrees=45
  twistMotion=Free
 }
 {bone1Name=whip_skin_15, bone2Name=whip_skin_14, swing1LimitDegrees=5, swing2LimitDegrees=5, twistLimitDegrees=0, twistMotion=Locked}
 {bone1Name=whip_skin_16, bone2Name=whip_skin_15, swing1LimitDegrees=5, swing2LimitDegrees=5, twistLimitDegrees=0, twistMotion=Locked}
]
